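# Additive unlit material that fakes a volumetric glow with a camera-facing sprite, nudged toward the camera via world position offset.
# A radial sphere mask forms the glow, with optional two-sine flicker phase-shifted by component pivot relative to actor position.
# Depth fade softens intersections, bypassed on ES3_1.
# A synced distance fade scales both sprite size and glow strength, with optional time-of-day scaling on top.
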Material /Game/Effects/Fort_Effects/Materials/Glows/M_FakeVolumeGlow_Sprite
graph {
 properties {
  bUsedWithMeshParticles=true
  bUsedWithParticleSprites=true
  blendMode=BLEND_Additive
  pixelDepthOffsetMode=PDOM_Legacy
  shadingModel=MSM_Unlit
  usedShadingModels=Unlit
 }
 nodes [
  n1 = ScalarParameter() {defaultValue=1000, desc=Size of the sprites, parameterName=Size}
  n2 = MaterialFunctionCall(
   Fade_Start_Distance=ScalarParameter() {defaultValue=5000, parameterName="Fade_Start_Distance (S)"}
   Fade_End_Distance=ScalarParameter() {defaultValue=3000, parameterName="Fade_End_Distance (S)"}
  ) {materialFunction=&MF_Distance_SyncedFade_FX}
  n3 = Subtract(A=CameraPositionWS(), B=WorldPosition().XYZ)
  Add(
   A=MaterialFunctionCall!(Center Location=WorldPosition().XYZ, 0-1 UVs, XY scale=Multiply(A=@n1, B=@n2.Result), OverridePivot, "Normals (Optional)") {materialFunction=&Sprite}.World position offset
   B=Multiply(A=Normalize(VectorInput=@n3), B=ScalarParameter() {defaultValue=20, parameterName=CameraOffsetAmount})
  )
  n4 = TextureCoordinate()
  n5 = Constant2Vector() {g=0.5, r=0.5}
  n6 = Constant() {r=0.49}
  n7 = Constant() {r=-1}
  n8 = MaterialFunctionCall(Float3=Subtract(
    A=MaterialFunctionCall() {materialFunction=&ComponentPivotLocation}.Result
    B=ActorPositionWS()
   )) {materialFunction=&BreakOutFloat3Components}
  n9 = Add(A=@n8.R, B=@n8.R)
  n10 = Add(A=Time(), B=Multiply(A=Add(A=@n9, B=@n8.B)))
  n11 = Multiply(A=@n10) {constB=1.351}
  n12 = Multiply(A=@n10) {constB=1.21}
  n13 = Add(A=Multiply(A=Sine(None=@n11)) {constB=0.25}, B=Multiply(A=Sine(None=@n12) {period=2}) {constB=0.4})
  n14 = Multiply(
   A=Power(Base=SphereMask(A=@n4, B=@n5, Radius=@n6, Hardness=@n7))
   B=StaticSwitchParameter(True=Add(A=@n13), False=Constant() {r=1}) {defaultValue=true, parameterName=UseFlicker?}
  )
  n15 = DepthFade(Opacity=@n14, FadeDistance=ScalarParameter() {parameterName=FadeDistance})
  n16 = Multiply(A=@n14, B=Constant() {r=1})
  n17 = Multiply(
   A=Multiply(
    A=VectorParameter() {parameterName=EmissiveColour, defaultValue=<rgba 5 1.6633 0.45 0.035>}.A
    B=FeatureLevelSwitch(Default=@n15, ES3_1=@n16)
   )
   B=VertexInterpolator(VS=@n2.Result).PS
  )
  n18 = ScalarParameter() {defaultValue=1, parameterName="Value_Morning (S)"}
  n19 = ScalarParameter() {defaultValue=1, parameterName="Value_Day (S)"}
  n20 = ScalarParameter() {defaultValue=1, parameterName="Value_Evening (S)"}
  n21 = ScalarParameter() {defaultValue=1, parameterName="Value_Night (S)"}
  StaticSwitchParameter(
   True=Multiply(
    A=@n17
    B=MaterialFunctionCall(Value_Morning=@n18, Value_Day=@n19, Value_Evening=@n20, Value_Night=@n21) {materialFunction=&Scalar_Time_of_Day}.Result
   )
   False=@n17
  ) {parameterName=UseTimeofDayOpacity?}
 ]
}
